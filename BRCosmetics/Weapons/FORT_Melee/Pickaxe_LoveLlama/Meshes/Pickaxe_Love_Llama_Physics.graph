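# Physics asset for the Love Llama two-handed pickaxe cosmetic, providing collision and ragdoll-style motion for its dangling flower piece.
# The handle bone carries a long capsule for the shaft and a box around the head.
# A small capsule body on the dyn_flower bone is constrained to the handle with 45-degree swing limits and almost no twist, so the flower can sway without spinning.
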PhysicsAsset /BRCosmetics/Weapons/FORT_Melee/Pickaxe_LoveLlama/Meshes/Pickaxe_Love_Llama_Physics
bodies {
 Melee_TwoHanded_Handle [
  {shapeType=Capsule, length=75.2599, radius=1.8863, center=<0.6648 -0.987 15.3385>, rotation=<rot 0.0507 -89.996 -0.0357>}
  {shapeType=Box, extentX=13.5, extentY=31.5, extentZ=34, center=<0.6294 5.0131 68.8513>, rotation=<rot 0 3.5495 0>}
 ]
 dyn_flower [{shapeType=Capsule, length=7, radius=4, center=<0 0 -1.2647>}]
}
constraints [{bone1Name=dyn_flower, bone2Name=Melee_TwoHanded_Handle, swing1LimitDegrees=45, swing2LimitDegrees=45, twistLimitDegrees=1}]
